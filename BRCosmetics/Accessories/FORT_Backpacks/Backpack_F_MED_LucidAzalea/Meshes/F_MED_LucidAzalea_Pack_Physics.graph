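# Physics asset for the LucidAzalea backpack cosmetic, defining simulation bodies and joints for its dangling cape.
# One large capsule covers the pack, with a small cape_top body jointed to it.
# Three cape chains (left, middle, right) hang from cape_top, each a small master body followed by three identical capsule segments.
# Segment joints use limited swing and tight twist, while the master and cape_top joints leave all angular motion free.
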PhysicsAsset /BRCosmetics/Accessories/FORT_Backpacks/Backpack_F_MED_LucidAzalea/Meshes/F_MED_LucidAzalea_Pack_Physics
bodies {
 cape_top [{shapeType=Capsule, length=0.505, radius=0.505, shapeName=cape_top_capsule, rotation=<rot 0 -90 0>}]
 dyn_cape_01_l [{shapeType=Capsule, length=3, radius=2.5, shapeName=dyn_cape_01_l_capsule, center=<2.3 0 0>, rotation=<rot 0 -4.2588 0>}]
 dyn_cape_01_mid [{shapeType=Capsule, length=3, radius=2.5, shapeName=dyn_cape_01_mid_capsule, center=<2.3 0 0>, rotation=<rot 0 -4.2588 0>}]
 dyn_cape_01_r [{shapeType=Capsule, length=3, radius=2.5, shapeName=dyn_cape_01_r_capsule, center=<2.3 0 0>, rotation=<rot 0 -4.2588 0>}]
 dyn_cape_02_l [{shapeType=Capsule, length=3, radius=2.5, shapeName=dyn_cape_02_l_capsule, center=<2.3 0 0>, rotation=<rot 0 -4.2588 0>}]
 dyn_cape_02_mid [{shapeType=Capsule, length=3, radius=2.5, shapeName=dyn_cape_02_mid_capsule, center=<2.3 0 0>, rotation=<rot 0 -4.2588 0>}]
 dyn_cape_02_r [{shapeType=Capsule, length=3, radius=2.5, shapeName=dyn_cape_02_r_capsule, center=<2.3 0 0>, rotation=<rot 0 -4.2588 0>}]
 dyn_cape_03_l [{shapeType=Capsule, length=3, radius=2.5, shapeName=dyn_cape_03_l_capsule, center=<2.3 0 0>, rotation=<rot 0 -4.2588 0>}]
 dyn_cape_03_mid [{shapeType=Capsule, length=3, radius=2.5, shapeName=dyn_cape_03_mid_capsule, center=<2.3 0 0>, rotation=<rot 0 -4.2588 0>}]
 dyn_cape_03_r [{shapeType=Capsule, length=3, radius=2.5, shapeName=dyn_cape_03_r_capsule, center=<2.3 0 0>, rotation=<rot 0 -4.2588 0>}]
 dyn_cape_mstr_l [{shapeType=Capsule, length=0.505, radius=0.505, shapeName=dyn_cape_mstr_l_capsule, rotation=<rot 0 -90 0>}]
 dyn_cape_mstr_mid [{shapeType=Capsule, length=0.505, radius=0.505, shapeName=dyn_cape_mstr_mid_capsule, rotation=<rot 0 -90 0>}]
 dyn_cape_mstr_r [{shapeType=Capsule, length=0.505, radius=0.505, shapeName=dyn_cape_mstr_r_capsule, rotation=<rot 0 -90 0>}]
 pack [{
   shapeType=Capsule
   length=9.6296
   radius=14.9597
   shapeName=pack_capsule
   center=<-0.0495 17.5308 -23.0594>
   rotation=<rot 90 -179.8508 178.97>
  }]
}
constraints [
 {bone1Name=dyn_cape_02_l, bone2Name=dyn_cape_01_l, swing1LimitDegrees=20, swing2LimitDegrees=50, twistLimitDegrees=5}
 {bone1Name=dyn_cape_03_l, bone2Name=dyn_cape_02_l, swing1LimitDegrees=20, swing2LimitDegrees=50, twistLimitDegrees=5}
 {bone1Name=dyn_cape_02_mid, bone2Name=dyn_cape_01_mid, swing1LimitDegrees=20, swing2LimitDegrees=50, twistLimitDegrees=5}
 {bone1Name=dyn_cape_03_mid, bone2Name=dyn_cape_02_mid, swing1LimitDegrees=20, swing2LimitDegrees=50, twistLimitDegrees=5}
 {bone1Name=dyn_cape_02_r, bone2Name=dyn_cape_01_r, swing1LimitDegrees=20, swing2LimitDegrees=50, twistLimitDegrees=5}
 {bone1Name=dyn_cape_03_r, bone2Name=dyn_cape_02_r, swing1LimitDegrees=20, swing2LimitDegrees=50, twistLimitDegrees=5}
 {bone1Name=dyn_cape_01_r, bone2Name=dyn_cape_mstr_r, swing1LimitDegrees=15, swing2LimitDegrees=50, twistLimitDegrees=5}
 {bone1Name=dyn_cape_01_l, bone2Name=dyn_cape_mstr_l, swing1LimitDegrees=15, swing2LimitDegrees=50, twistLimitDegrees=5}
 {bone1Name=dyn_cape_01_mid, bone2Name=dyn_cape_mstr_mid, swing1LimitDegrees=15, swing2LimitDegrees=50, twistLimitDegrees=5}
 {
  bone1Name=dyn_cape_mstr_l
  bone2Name=cape_top
  swing1LimitDegrees=45
  swing1Motion=Free
  swing2LimitDegrees=45
  swing2Motion=Free
  twistLimitDegrees=45
  twistMotion=Free
 }
 {
  bone1Name=dyn_cape_mstr_mid
  bone2Name=cape_top
  swing1LimitDegrees=45
  swing1Motion=Free
  swing2LimitDegrees=45
  swing2Motion=Free
  twistLimitDegrees=45
  twistMotion=Free
 }
 {
  bone1Name=dyn_cape_mstr_r
  bone2Name=cape_top
  swing1LimitDegrees=45
  swing1Motion=Free
  swing2LimitDegrees=45
  swing2Motion=Free
  twistLimitDegrees=45
  twistMotion=Free
 }
 {
  bone1Name=cape_top
  bone2Name=pack
  swing1LimitDegrees=45
  swing1Motion=Free
  swing2LimitDegrees=45
  swing2Motion=Free
  twistLimitDegrees=45
  twistMotion=Free
 }
]
